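# Physics asset for the director's clapboard skeletal mesh, a movie set prop under BRCosmetics.
# It has a single body on the Melee_TwoHanded_Handle bone, built from one box shape offset upward from the bone.
# No constraints are defined, so the prop simulates as one rigid piece with no jointed parts.
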PhysicsAsset /BRCosmetics/Environments/Props/MovieSet/Meshes/SK_Directors_ClapBoard_Physics
bodies {Melee_TwoHanded_Handle [{shapeType=Box, extentX=41.4375, extentY=11.1424, extentZ=33.7664, center=<0.0511 -1.88 16.7284>}]}
constraints []
